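# Unlit translucent material for spike sprites in the Niagara effects of the PirouetteWeld pickaxe cosmetic.
# Particle color RGB feeds the emissive input of the shared MF_FX_EmissiveAdjustments function, so the effect supplies the tint.
# The red channel of the T_FX_Spike texture is multiplied by particle alpha to form a fadeable spike-shaped mask.
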Material /BRCosmetics/Weapons/FORT_Melee/Pickaxe_PirouetteWeld/FX/M_FX_Spike
graph {
 properties {bUsedWithNiagaraSprites=true, blendMode=BLEND_Translucent, pixelDepthOffsetMode=PDOM_Legacy, shadingModel=MSM_Unlit, usedShadingModels=Unlit}
 nodes [
  n1 = ParticleColor()
  MaterialFunctionCall(Emissive=@n1.RGB) {materialFunction=&MF_FX_EmissiveAdjustments}
  Multiply(
   A=TextureSample() {samplerType=SAMPLERTYPE_LinearColor, texture=&/T_FX_Spike}.R
   B=@n1.A
  )
 ]
}
