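# Physics asset for the PirouetteWeld pickaxe, a two-handed melee weapon with two dangling bow bones.
# The handle bone carries two capsules and a box as its collision shapes.
# The left and right bow bones each get a small sideways capsule and a constraint to the handle.
# Both constraints share angular limits that allow wide swing on one axis, narrower swing on the other and almost no twist.
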PhysicsAsset /BRCosmetics/Weapons/FORT_Melee/Pickaxe_PirouetteWeld/Meshes/PirouetteWeld_Axe_Physics
bodies {
 Melee_TwoHanded_Handle [
  {
   shapeType=Capsule
   length=66.5682
   radius=3.8435
   shapeName=Melee_TwoHanded_Handle_capsule
   center=<0.0489 -1.0328 46.4515>
   rotation=<rot -0.0087 -180 -0.1027>
  }
  {shapeType=Capsule, length=24, radius=6, shapeName=Melee_TwoHanded_Handle_capsule, center=<0.3661 -0.9151 60.617>}
  {shapeType=Box, extentX=1, extentY=8.5, extentZ=7.5, shapeName=Melee_TwoHanded_Handle_box, center=<1.6713 -1.2891 86.7529>}
 ]
 dyn_bow_l [{shapeType=Capsule, length=10, radius=2, shapeName=dyn_ribbon_2_RGT_capsule, center=<2 0 0>, rotation=<rot 0 0 90>}]
 dyn_bow_r [{shapeType=Capsule, length=10, radius=2, shapeName=dyn_ribbon_2_RGT_capsule, center=<2 0 0>, rotation=<rot 0 0 90>}]
}
constraints [
 {bone1Name=dyn_bow_r, bone2Name=Melee_TwoHanded_Handle, swing1LimitDegrees=10, swing2LimitDegrees=50, twistLimitDegrees=2}
 {bone1Name=dyn_bow_l, bone2Name=Melee_TwoHanded_Handle, swing1LimitDegrees=10, swing2LimitDegrees=50, twistLimitDegrees=2}
]
